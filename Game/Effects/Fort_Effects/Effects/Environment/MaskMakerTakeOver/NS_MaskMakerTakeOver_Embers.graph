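# GPU Niagara effect of drifting, glowing purple-pink embers for the MaskMakerTakeOver environment event.
# Motes_Epic importance-samples a mask texture to spawn in its pattern, while Motes_Epic001 and Motes_High fill boxes.
# Curl noise and upward wind move motes under speed caps, with distance-field avoidance only on Motes_High.
# A region mask scales size and alpha; Epic emitters add waveform flicker and second sprite renderers, computing dimmed color and enlarged size attributes.
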NiagaraSystem /Game/Effects/Fort_Effects/Effects/Environment/MaskMakerTakeOver/NS_MaskMakerTakeOver_Embers
dependencies {
 usedDataInterfaces [
  /Script/Niagara.NiagaraDataInterfaceTexture
  &NiagaraDataInterfaceCurve
  &NiagaraDataInterfaceCollisionQuery
  &NiagaraDataInterfaceLandscape
  &NiagaraDataInterfaceColorCurve
 ]
 usedDynamicInputs []
 usedModules [&SystemState]
}
emitters {
 Motes_Epic {
  simTarget=GPUComputeSim
  emitterUpdateScript [
   &EmitterState()
   &SpawnRate(SpawnRate=1000, Use Spawn Probability=false)
  ]
  particleSpawnScript [
   &InitializeParticle(
    Color=&RandomRangeLinearColor(Minimum=<rgba 0.2417 0 0.3958 0.5>, Maximum=<rgba 1 0 0.605 0.5>)
    Lifetime=&RandomRangeFloat(Minimum=2, Maximum=3)
    Sprite Size Max=<4 6>
    Sprite Size Min=<3 4>
    Sprite Size Mode=::NewEnumerator2
    UsePositionOffset=false
    Write Material Random=true
   )
   /Niagara/Modules/Update/Textures/SampleTexture(
    Apply Sampled RGBA to Particle Color=false
    EnableFalloff=true
    Generate UV Using Importance Sampling=true
    Importance Texture=DataInterface() {propertyValues {
      Texture=/Game/Effects/Fort_Effects/Textures/MaskMakerTakeOver/T_MaskMakerTakeOver_NiagaraMask_01
      TextureUserParameter {parameter {name=None, typeDefHandle {registeredTypeIndex=2}, varData []}}
     }}
    Rotation Angle=0.25
    Rotation Axis:Vector3f=<0 0 1>
    Rotation Mode:ENiagara_RotationMode=::NewEnumerator1
    Texture=DataInterface() {propertyValues {
      Texture=&T_MaskMakerTakeOver_NiagaraMask_01
      TextureUserParameter {parameter {name=None, typeDefHandle {registeredTypeIndex=2}, varData []}}
     }}
    Transform UV Into Position=true
    UV:Vector2f=&RandomRangeVector2D(Minimum=<0 0>, Maximum=<1 1>)
    UV Origin:NiagaraPosition=&AddVectorToPosition(Position=$Engine.Emitter.SimulationPosition, Vector=<-35 50 -25>)
    UV Scale:Vector2f=<2038.9437 2038.9437>
   )
   &ShapeLocation(
    Box Midpoint=<0.5 0.5 0>
    Box Size=<2000 2000 1>
    MaskSpawnGroup=false
    Offset=<-50 25 -25>
    Offset Mode=::NewEnumerator0
    Override Seed=false
    Shape Primitive=::NewEnumerator2
    Surface Only=false
    UseOwnerScale=true
   ) {enabled=false}
  ]
  particleUpdateScript [
   &ParticleState()
   &CurlNoiseForce(
    Bias Noise=false
    Noise Frequency=&RandomRangeFloat(Minimum=5, Maximum=15)
    Noise Strength=&FloatFromCurve(
     FloatCurve=DataInterface() {propertyValues {Curve {
        defaultValue=1
        keys [
         {interpMode=RCIM_Cubic, tangentMode=RCTM_User, time=0.25, value=0}
         {arriveTangent=3.0103, interpMode=RCIM_Cubic, leaveTangent=3.0103, tangentMode=RCTM_User, time=1, value=1}
        ]
       }}}
     Scale Curve=500
    )
    Pan Noise=false
   )
   /Niagara/Modules/Math/AvoidDistanceFieldSurfaces_GPU(Avoid Nearest Surface=true, Avoid Oncoming Surfaces=true, Falloff Distance=1500, Nearest Surface Avoidance Strength=250) {enabled=false}
   &WindForce(
    Scale Min / Max:Vector2f=<2 0.05>
    Speed Min / Max:Vector2f=<0 25>
    Turbulence Flow Loop Duration=2
    Turbulence Scale Mode:ENiagaraWindTurbulenceContributionMode=::NewEnumerator1
    Wind Speed=<0 0 100>
   )
   &Drag(Drag=2, Use Linear Drag=true, Use Rotational Drag=true)
   &SolveForcesAndVelocity(Clamp Velocity=true, Limit Acceleration=false, Speed Limit=25)
   &ScaleSpriteSizeBySpeed(Max Scale Factor=<1 1>, Min Scale Factor=<0.25 0.25>, Velocity Threshold=100)
   &RegionMask(Offset:Vector3f=<0 0 -1200>, Region Falloff=1, Region Radius=1800, UseOffset=true, UseRotation=false)
   &ScaleColor(
    Scale Alpha=&FloatFromCurve(
     FloatCurve=DataInterface() {propertyValues {Curve {
        defaultValue=1
        keys [
         {interpMode=RCIM_Cubic, tangentMode=RCTM_User, time=0, value=0}
         {interpMode=RCIM_Cubic, tangentMode=RCTM_User, time=0.5, value=1}
         {tangentMode=RCTM_User, time=1, value=0}
        ]
       }}}
     Scale Curve=&Multiply_Float(
      A=$Particles.RegionMask.Float
      B=&Waveform(
       Waveform Count=::NewEnumerator1
       "Waveform [1]"=::NewEnumerator0
       "[1] Frequency"=&RandomRangeFloat(Minimum=0.5, Maximum=1)
       "[1] Amplitude Min/Max"=<0 1>
       "[2] Average"=false
       "[2] Input Position"=0
       "[2] Phase"=$Emitter.Age
       "[2] Phase Scale"=1
       "[2] Compound Frequency"=2
       "[2] Frequency"=&RandomRangeFloat(Minimum=1.25, Maximum=2.5)
       "[2] Amplitude Min/Max"=<0.5 1>
       "[2] Scale"=1
       "[2] Bias"=0
       "[2] Exponent"=1
       "[2] Pulse Duration"=0.1
       "[2] Clip Triangle Min/Max"=<0 0>
       "[2] Random Seed"=0
      )
     )
    )
    Scale RGB=<25 1 25>
    ScaleA=true
    ScaleRGB=true
   )
   &DynamicMaterialParameters(
    Index 0 Param 1=2
    Index 0 Param 2=500
    Index 0 Param 3=500
    Index 0 Param 4=&Multiply_Float(
     A=$Particles.RegionMask.Float
     B=&Waveform(
      Global Amplitude Scale=2500
      "Waveform [1]"=::NewEnumerator0
      "[1] Phase Scale"=&RandomRangeFloat(Minimum=0.5, Maximum=1)
      "[1] Frequency"=&RandomRangeFloat(Minimum=0.4, Maximum=1)
      "[1] Amplitude Min/Max"=<0 1>
     )
    )
    Module.Index0Param0Write=true
    Module.Index0Param1Write=true
    Module.Index0Param2Write=true
    Module.Index0Param3Write=true
   )
   &ScaleSpriteSize(
    Scale Sprite Size Mode=::NewEnumerator0
    Uniform Scale Factor=&Multiply_Float(A=0.3, B=$Particles.RegionMask.Float)
   )
   /SetVariables(
    Particles.Linear Color=&Multiply_LinearColorByFloat(Linear Color=$Particles.Color, Float=0.15)
    Particles.Vector 2D:Vector2f=&Multiply_Vector2D_ByFloat(Float=15, Vector2D=$Particles.SpriteSize)
   ) {isSetParameters=true, module=SetVariables_#1}
   /Niagara/Modules/Update/Camera/MaintainInCameraParticleScale(Pixel Depth=0.1) {enabled=false}
  ]
  rendererClasses [&NiagaraSpriteRendererProperties]
  renderers [
   {rendererClass=&NiagaraSpriteRendererProperties}
   {rendererClass=&NiagaraSpriteRendererProperties, rendererIndex=1}
  ]
 }
 Motes_Epic001 {
  simTarget=GPUComputeSim
  emitterUpdateScript [
   &EmitterState()
   &SpawnRate(SpawnRate=500, Use Spawn Probability=false)
  ]
  particleSpawnScript [
   &InitializeParticle(
    Color=&RandomRangeLinearColor(Minimum=<rgba 0.2417 0 0.3958 0.5>, Maximum=<rgba 1 0 0.605 0.5>)
    Lifetime=&RandomRangeFloat(Minimum=2, Maximum=3)
    Sprite Size Max=<2 6>
    Sprite Size Min=<1 2>
    Sprite Size Mode=::NewEnumerator2
    UsePositionOffset=false
    Write Material Random=true
   )
   &ShapeLocation(
    Box Midpoint=<0.5 0.5 0>
    Box Size=<1500 1500 250>
    MaskSpawnGroup=false
    Offset=<-50 25 0>
    Offset Mode=::NewEnumerator0
    Override Seed=false
    Shape Primitive=::NewEnumerator2
    Surface Only=false
    UseOwnerScale=true
   )
  ]
  particleUpdateScript [
   &ParticleState()
   &CurlNoiseForce(
    Bias Noise=false
    Noise Frequency=&RandomRangeFloat(Minimum=5, Maximum=15)
    Noise Strength=&FloatFromCurve(
     FloatCurve=DataInterface() {propertyValues {Curve {
        defaultValue=1
        keys [
         {interpMode=RCIM_Cubic, tangentMode=RCTM_User, time=0.25, value=0}
         {arriveTangent=3.0103, interpMode=RCIM_Cubic, leaveTangent=3.0103, tangentMode=RCTM_User, time=1, value=1}
        ]
       }}}
     Scale Curve=500
    )
    Pan Noise=false
   )
   &AvoidDistanceFieldSurfaces_GPU(Avoid Nearest Surface=true, Avoid Oncoming Surfaces=true, Falloff Distance=1500, Nearest Surface Avoidance Strength=250) {enabled=false}
   &WindForce(
    Scale Min / Max:Vector2f=<2 0.05>
    Speed Min / Max:Vector2f=<0 25>
    Turbulence Flow Loop Duration=2
    Turbulence Scale Mode:ENiagaraWindTurbulenceContributionMode=::NewEnumerator1
    Wind Speed=<0 0 100>
   )
   &Drag(Drag=2, Use Linear Drag=true, Use Rotational Drag=true)
   &SolveForcesAndVelocity(Clamp Velocity=true, Limit Acceleration=false, Speed Limit=25)
   &ScaleSpriteSizeBySpeed(Max Scale Factor=<1 1>, Min Scale Factor=<0.25 0.25>, Velocity Threshold=100)
   &RegionMask(Offset:Vector3f=<0 0 -1200>, Region Falloff=1, Region Radius=1800, UseOffset=true, UseRotation=false)
   &ScaleColor(
    Scale Alpha=&FloatFromCurve(
     FloatCurve=DataInterface() {propertyValues {Curve {
        defaultValue=1
        keys [
         {interpMode=RCIM_Cubic, tangentMode=RCTM_User, time=0, value=0}
         {interpMode=RCIM_Cubic, tangentMode=RCTM_User, time=0.5, value=1}
         {tangentMode=RCTM_User, time=1, value=0}
        ]
       }}}
     Scale Curve=&Multiply_Float(
      A=$Particles.RegionMask.Float
      B=&Waveform(
       Waveform Count=::NewEnumerator1
       "Waveform [1]"=::NewEnumerator0
       "[1] Frequency"=&RandomRangeFloat(Minimum=0.5, Maximum=1)
       "[1] Amplitude Min/Max"=<0 1>
       "[2] Average"=false
       "[2] Input Position"=0
       "[2] Phase"=$Emitter.Age
       "[2] Phase Scale"=1
       "[2] Compound Frequency"=2
       "[2] Frequency"=&RandomRangeFloat(Minimum=1.25, Maximum=2.5)
       "[2] Amplitude Min/Max"=<0.5 1>
       "[2] Scale"=1
       "[2] Bias"=0
       "[2] Exponent"=1
       "[2] Pulse Duration"=0.1
       "[2] Clip Triangle Min/Max"=<0 0>
       "[2] Random Seed"=0
      )
     )
    )
    Scale RGB=<25 1 25>
    ScaleA=true
    ScaleRGB=true
   )
   &DynamicMaterialParameters(
    Index 0 Param 1=2
    Index 0 Param 2=500
    Index 0 Param 3=500
    Index 0 Param 4=&Multiply_Float(
     A=$Particles.RegionMask.Float
     B=&Waveform(
      Global Amplitude Scale=2500
      "Waveform [1]"=::NewEnumerator0
      "[1] Phase Scale"=&RandomRangeFloat(Minimum=0.5, Maximum=1)
      "[1] Frequency"=&RandomRangeFloat(Minimum=0.4, Maximum=1)
      "[1] Amplitude Min/Max"=<0 1>
     )
    )
    Module.Index0Param0Write=true
    Module.Index0Param1Write=true
    Module.Index0Param2Write=true
    Module.Index0Param3Write=true
   )
   &ScaleSpriteSize(
    Scale Sprite Size Mode=::NewEnumerator0
    Uniform Scale Factor=&Multiply_Float(A=0.3, B=$Particles.RegionMask.Float)
   )
   /SetVariables(
    Particles.Linear Color=&Multiply_LinearColorByFloat(Linear Color=$Particles.Color, Float=0.15)
    Particles.Vector 2D:Vector2f=&Multiply_Vector2D_ByFloat(Float=15, Vector2D=$Particles.SpriteSize)
   ) {isSetParameters=true, module=SetVariables_#1}
   &MaintainInCameraParticleScale(Pixel Depth=0.1) {enabled=false}
  ]
  rendererClasses [&NiagaraSpriteRendererProperties]
  renderers [
   {rendererClass=&NiagaraSpriteRendererProperties}
   {rendererClass=&NiagaraSpriteRendererProperties, rendererIndex=1}
  ]
 }
 Motes_High {
  simTarget=GPUComputeSim
  emitterUpdateScript [
   &EmitterState()
   &SpawnRate(SpawnRate=1000, Use Spawn Probability=false)
  ]
  particleSpawnScript [
   &InitializeParticle(
    Color=&RandomRangeLinearColor(Minimum=<rgba 0.2417 0 0.3958 0.5>, Maximum=<rgba 1 0 0.605 0.5>)
    Lifetime=5
    Sprite Size Max=<4 4>
    Sprite Size Min=<3 3>
    Sprite Size Mode=::NewEnumerator2
    UsePositionOffset=false
    Write Material Random=true
   )
   &ShapeLocation(
    Box Midpoint=<0.5 0.5 0>
    Box Size=<1500 1500 600>
    MaskSpawnGroup=false
    Override Seed=false
    Shape Primitive=::NewEnumerator2
    Surface Only=false
    UseOwnerScale=true
   )
  ]
  particleUpdateScript [
   &ParticleState()
   &CurlNoiseForce(
    Bias Noise=false
    Noise Frequency=&RandomRangeFloat(Minimum=0, Maximum=1)
    Noise Strength=250
    Pan Noise=false
   )
   &AvoidDistanceFieldSurfaces_GPU(Avoid Nearest Surface=true, Avoid Oncoming Surfaces=true, Falloff Distance=1500, Nearest Surface Avoidance Strength=250)
   &WindForce(
    Scale Min / Max:Vector2f=<5 0.25>
    Speed Min / Max:Vector2f=<0 100>
    Turbulence Flow Loop Duration=2
    Turbulence Scale Mode:ENiagaraWindTurbulenceContributionMode=::NewEnumerator1
    Wind Speed=<0 0 100>
   )
   &Drag(Use Linear Drag=true, Use Rotational Drag=true)
   &SolveForcesAndVelocity(Clamp Velocity=true, Limit Acceleration=false, Speed Limit=15)
   &RegionMask(Offset:Vector3f=<0 0 -1200>, Region Falloff=1, Region Radius=1800, UseOffset=true, UseRotation=false)
   &ScaleColor(
    Scale Alpha=&FloatFromCurve(
     FloatCurve=DataInterface() {propertyValues {Curve {
        defaultValue=1
        keys [
         {interpMode=RCIM_Cubic, tangentMode=RCTM_User, time=0, value=0}
         {interpMode=RCIM_Cubic, tangentMode=RCTM_User, time=0.5, value=1}
         {tangentMode=RCTM_User, time=1, value=0}
        ]
       }}}
     Scale Curve=&Multiply_Float(A=$Particles.RegionMask.Float, B=1)
    )
    ScaleA=true
    ScaleRGB=true
   )
   &DynamicMaterialParameters(
    Index 0 Param 1=2
    Index 0 Param 2=500
    Index 0 Param 3=500
    Index 0 Param 4=&Multiply_Float(
     A=$Particles.RegionMask.Float
     B=&Waveform(
      Global Amplitude Scale=2500
      "Waveform [1]"=::NewEnumerator0
      "[1] Phase Scale"=&RandomRangeFloat(Minimum=0.5, Maximum=1)
      "[1] Frequency"=&RandomRangeFloat(Minimum=0.4, Maximum=1)
      "[1] Amplitude Min/Max"=<0 1>
     )
    )
    Module.Index0Param0Write=true
    Module.Index0Param1Write=true
    Module.Index0Param2Write=true
    Module.Index0Param3Write=true
   )
   &ScaleSpriteSize(
    Scale Sprite Size Mode=::NewEnumerator0
    Uniform Scale Factor=&Multiply_Float(A=0.3, B=$Particles.RegionMask.Float)
   )
   &MaintainInCameraParticleScale(Pixel Depth=0.1) {enabled=false}
  ]
  rendererClasses [&NiagaraSpriteRendererProperties]
  renderers [{rendererClass=&NiagaraSpriteRendererProperties}]
 }
}
summary {
 systemName=NS_MaskMakerTakeOver_Embers
 emitters [
  {emitterName=Motes_Epic, simTarget=GPUComputeSim, rendererClasses [&NiagaraSpriteRendererProperties]}
  {emitterName=Motes_High, simTarget=GPUComputeSim, rendererClasses [&NiagaraSpriteRendererProperties]}
  {emitterName=Motes_Epic001, simTarget=GPUComputeSim, rendererClasses [&NiagaraSpriteRendererProperties]}
 ]
 userVariables []
}
