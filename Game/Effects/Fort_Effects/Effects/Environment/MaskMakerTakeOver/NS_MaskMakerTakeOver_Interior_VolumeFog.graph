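# Niagara volume fog system for the MaskMakerTakeOver interior environment, with a single emitter named Empty that is saved disabled.
# The emitter steadily spawns short-lived pink-magenta particles in a flat box scaled by User.SpawnVolumeScale, with two sprite renderers and a mesh renderer.
# Distance-based LOD remaps grow mesh size and raise opacity farther away, combined with over-life curves that pulse scale and fade alpha.
# User.SpawnRateScale is exposed, but the shown spawn rate does not use it.
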NiagaraSystem /Game/Effects/Fort_Effects/Effects/Environment/MaskMakerTakeOver/NS_MaskMakerTakeOver_Interior_VolumeFog
dependencies {
 usedDataInterfaces [&NiagaraDataInterfaceCurve, &NiagaraDataInterfaceColorCurve]
 usedDynamicInputs []
 usedModules [&SystemState]
}
emitters {Empty {
  bEnabled=false
  emitterUpdateScript [
   &EmitterState()
   &SpawnRate(SpawnRate=6, Use Spawn Probability=false)
  ]
  particleSpawnScript [
   &InitializeParticle(
    Color=<rgba 1 0.1128 0.8059 1>
    Lifetime=&RandomRangeFloat(Minimum=3, Maximum=5)
    Mesh Scale Max=<8 8 8>
    Mesh Scale Min=<6 6 6>
    Mesh Scale Mode=::NewEnumerator2
    Sprite Size Mode=::NewEnumerator1
    Uniform Sprite Size Max=350
    Uniform Sprite Size Min=250
    UsePositionOffset=false
    Write Material Random=true
   )
   &ShapeLocation(
    Box Size=&Multiply_VectorByFloat(Vector=<1500 1500 1>, Float=$User.SpawnVolumeScale)
    MaskSpawnGroup=false
    Override Seed=false
    Shape Primitive=::NewEnumerator2
    Surface Only=false
    UseOwnerScale=true
   )
  ]
  particleUpdateScript [
   &ParticleState()
   &ScaleMeshSize(Scale Factor=&Multiply_VectorByFloat(
     Vector=&Lerp_Vector(
      A=<0.125 0.125 0.125>
      B=<1 1 1>
      Alpha=&FN_LOD_Distance_Remap(Start Distance=500, End Distance=3000, Minimum Output=0, Maximum Output=1.25, Clamp Output=true)
     )
     Float=&FloatFromCurve(FloatCurve=DataInterface() {propertyValues {Curve {
         defaultValue=1
         keys [
          {interpMode=RCIM_Cubic, tangentMode=RCTM_User, time=0, value=0.725}
          {interpMode=RCIM_Cubic, tangentMode=RCTM_User, time=0.5, value=1}
          {tangentMode=RCTM_User, time=1, value=0.8}
         ]
        }}})
    ))
   &ScaleColor(
    Scale Alpha=&FloatFromCurve(
     FloatCurve=DataInterface() {propertyValues {Curve {keys [
         {interpMode=RCIM_Cubic, tangentMode=RCTM_User, time=0, value=0}
         {interpMode=RCIM_Cubic, tangentMode=RCTM_User, time=0.5, value=1}
         {tangentMode=RCTM_User, time=1, value=0}
        ]}}}
     Scale Curve=&FN_LOD_Distance_Remap(Start Distance=1500, End Distance=5000, Minimum Output=0.02, Maximum Output=0.1, Clamp Output=true)
    )
    Scale RGB=<0.25 0.25 0.25>
    ScaleA=true
    ScaleRGB=true
   )
   &SolveForcesAndVelocity(Clamp Velocity=false, Limit Acceleration=false)
  ]
  rendererClasses [&NiagaraSpriteRendererProperties, &NiagaraMeshRendererProperties]
  renderers [
   {rendererClass=&NiagaraSpriteRendererProperties}
   {rendererClass=&NiagaraMeshRendererProperties, rendererIndex=1}
   {rendererClass=&NiagaraSpriteRendererProperties, rendererIndex=2}
  ]
 }}
summary {
 systemName=NS_MaskMakerTakeOver_Interior_VolumeFog
 emitters [{
   bEnabled=false
   emitterName=Empty
   rendererClasses [&NiagaraSpriteRendererProperties, &NiagaraMeshRendererProperties]
  }]
 userVariables [
  User.SpawnRateScale {type=NiagaraFloat, defaultValue=1, description=""}
  User.SpawnVolumeScale {type=NiagaraFloat, defaultValue=1, description=""}
 ]
}
